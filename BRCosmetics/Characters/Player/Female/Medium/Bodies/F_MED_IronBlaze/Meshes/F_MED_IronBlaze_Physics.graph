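# Physics asset for the F_MED_IronBlaze female medium player body, covering four belt pouches for secondary motion.
# Capsules on the pelvis and both thighs form the body volumes, with thighs constrained to the pelvis.
# Each pouch chains a tiny master anchor capsule, rotationally free from the pelvis, to a larger dyn capsule.
# Dyn-to-master joints use asymmetric swing limits and zero twist limit, with a tighter swing on the right front pouch.
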PhysicsAsset /BRCosmetics/Characters/Player/Female/Medium/Bodies/F_MED_IronBlaze/Meshes/F_MED_IronBlaze_Physics
bodies {
 dyn_L_back_belt_pouch_jnt [{
   shapeType=Capsule
   length=5.9508
   radius=5.5626
   shapeName=dyn_L_back_belt_pouch_jnt_capsule
   center=<3.2796 -0.0957 5.1135>
   rotation=<rot 90 0 9.9998>
  }]
 dyn_L_front_belt_pouch_jnt [{
   shapeType=Capsule
   length=13.684
   radius=4.7253
   shapeName=dyn_L_front_belt_pouch_jnt_capsule
   center=<6.3176 0.6135 4.2489>
   rotation=<rot 90 -179.9406 -179.9407>
  }]
 dyn_R_back_belt_pouch_jnt [{
   shapeType=Capsule
   length=19.1374
   radius=4.1254
   shapeName=dyn_R_back_belt_pouch_jnt_capsule
   center=<10.9386 0 -1.1065>
   rotation=<rot 80 180 0>
  }]
 dyn_R_front_belt_pouch_jnt [{
   shapeType=Capsule
   length=5.2688
   radius=3.6537
   shapeName=dyn_R_front_belt_pouch_jnt_capsule
   center=<3.507 0 -3.7204>
   rotation=<rot 80 0.0001 -0.0001>
  }]
 master_L_back_belt_pouch_jnt [{shapeType=Capsule, length=0.505, radius=0.505, shapeName=master_L_back_belt_pouch_jnt_capsule, rotation=<rot 0 -90 0>}]
 master_L_front_belt_pouch_jnt [{shapeType=Capsule, length=0.505, radius=0.505, shapeName=master_L_front_belt_pouch_jnt_capsule, rotation=<rot 0 -90 0>}]
 master_R_back_belt_pouch_jnt [{shapeType=Capsule, length=0.505, radius=0.505, shapeName=master_R_back_belt_pouch_jnt_capsule, rotation=<rot 0 -90 0>}]
 master_R_front_belt_pouch_jnt [{shapeType=Capsule, length=0.505, radius=0.505, shapeName=master_R_front_belt_pouch_jnt_capsule, rotation=<rot 0 -90 0>}]
 pelvis [{shapeType=Capsule, length=10.0684, radius=11.7862, center=<-6.0695 -2.3404 -0.8979>, rotation=<rot -0.6913 -90 1.2195>}]
 thigh_l [{shapeType=Capsule, length=26.3318, radius=7.0854, center=<-26.8952 -0.3768 -0.1827>, rotation=<rot 0 -84.1071 -90.996>}]
 thigh_r [{shapeType=Capsule, length=26.8479, radius=7.918, center=<21.0211 1.3286 -0.6889>, rotation=<rot 0.0642 -91.808 -89.6361>}]
}
constraints [
 {bone1Name=thigh_l, bone2Name=pelvis, swing1LimitDegrees=45, swing2LimitDegrees=45, twistLimitDegrees=45}
 {bone1Name=thigh_r, bone2Name=pelvis, swing1LimitDegrees=45, swing2LimitDegrees=45, twistLimitDegrees=45}
 {
  bone1Name=dyn_R_front_belt_pouch_jnt
  bone2Name=master_R_front_belt_pouch_jnt
  swing1LimitDegrees=25
  swing2LimitDegrees=15
  twistLimitDegrees=0
 }
 {
  bone1Name=master_L_back_belt_pouch_jnt
  bone2Name=pelvis
  swing1LimitDegrees=45
  swing1Motion=Free
  swing2LimitDegrees=45
  swing2Motion=Free
  twistLimitDegrees=45
  twistMotion=Free
 }
 {
  bone1Name=dyn_L_back_belt_pouch_jnt
  bone2Name=master_L_back_belt_pouch_jnt
  swing1LimitDegrees=45
  swing2LimitDegrees=15
  twistLimitDegrees=0
 }
 {
  bone1Name=master_L_front_belt_pouch_jnt
  bone2Name=pelvis
  swing1LimitDegrees=45
  swing1Motion=Free
  swing2LimitDegrees=45
  swing2Motion=Free
  twistLimitDegrees=45
  twistMotion=Free
 }
 {
  bone1Name=dyn_L_front_belt_pouch_jnt
  bone2Name=master_L_front_belt_pouch_jnt
  swing1LimitDegrees=45
  swing2LimitDegrees=15
  twistLimitDegrees=0
 }
 {
  bone1Name=master_R_back_belt_pouch_jnt
  bone2Name=pelvis
  swing1LimitDegrees=45
  swing1Motion=Free
  swing2LimitDegrees=45
  swing2Motion=Free
  twistLimitDegrees=45
  twistMotion=Free
 }
 {
  bone1Name=dyn_R_back_belt_pouch_jnt
  bone2Name=master_R_back_belt_pouch_jnt
  swing1LimitDegrees=45
  swing2LimitDegrees=15
  twistLimitDegrees=0
 }
 {
  bone1Name=master_R_front_belt_pouch_jnt
  bone2Name=pelvis
  swing1LimitDegrees=45
  swing1Motion=Free
  swing2LimitDegrees=45
  swing2Motion=Free
  twistLimitDegrees=45
  twistMotion=Free
 }
]
