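# Physics asset for the IronBlaze medium female character's FaceAcc cosmetic part, setting up dynamic motion for a hair bun and its loose tangles.
# Small capsules on head, master hair and bun joints form a chain linked by free-rotation constraints.
# Three larger tangle capsules (upper, lower left, lower right) each attach directly to the bun.
# Those tangle constraints use narrow, asymmetric swing limits and zero-degree twist limits, differing per tangle.
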PhysicsAsset /BRCosmetics/Characters/Player/Female/Medium/Bodies/F_MED_IronBlaze/Meshes/Parts/F_MED_IronBlaze_FaceAcc_Physics
bodies {
 dyn_L_lower_tangle_motion_jnt [{shapeType=Capsule, length=6.25, radius=3, shapeName=dyn_R_lower_tangle_motion_jnt_capsule, center=<1.7974 0 0.502>}]
 dyn_R_lower_tangle_motion_jnt [{shapeType=Capsule, length=3, radius=3, shapeName=dyn_R_lower_tangle_motion_jnt_capsule, center=<1.7974 0 0>}]
 dyn_upper_tangle_motion_jnt [{shapeType=Capsule, length=3, radius=3, shapeName=dyn_upper_tangle_motion_jnt_capsule, center=<3.1977 0 0>}]
 head [{shapeType=Capsule, length=1, radius=1, shapeName=head_capsule}]
 master_bun_jnt [{shapeType=Capsule, length=1, radius=1, shapeName=master_bun_jnt_capsule}]
 master_hair_jnt [{shapeType=Capsule, length=1, radius=1, shapeName=master_hair_jnt_capsule}]
}
constraints [
 {
  bone1Name=master_bun_jnt
  bone2Name=master_hair_jnt
  swing1LimitDegrees=45
  swing1Motion=Free
  swing2LimitDegrees=45
  swing2Motion=Free
  twistLimitDegrees=45
  twistMotion=Free
 }
 {
  bone1Name=master_hair_jnt
  bone2Name=head
  swing1LimitDegrees=45
  swing1Motion=Free
  swing2LimitDegrees=45
  swing2Motion=Free
  twistLimitDegrees=45
  twistMotion=Free
 }
 {bone1Name=dyn_upper_tangle_motion_jnt, bone2Name=master_bun_jnt, swing1LimitDegrees=15, swing2LimitDegrees=5, twistLimitDegrees=0}
 {bone1Name=dyn_R_lower_tangle_motion_jnt, bone2Name=master_bun_jnt, swing1LimitDegrees=5, swing2LimitDegrees=15, twistLimitDegrees=0}
 {bone1Name=dyn_L_lower_tangle_motion_jnt, bone2Name=master_bun_jnt, swing1LimitDegrees=25, swing2LimitDegrees=15, twistLimitDegrees=0}
]
